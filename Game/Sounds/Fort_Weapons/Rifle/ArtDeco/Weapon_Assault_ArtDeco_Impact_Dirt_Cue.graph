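# Sound cue for Art Deco assault rifle bullet impacts on dirt, though its samples are medium stone impacts.
# A Random node picks among eight equally weighted waves, feeding a Modulator that varies pitch and volume, then a Delay at default settings.
# It uses bullet impact attenuation, the impacts sound class and a rifle impact concurrency set, plus a concurrency override that prevents new sounds at the cap.
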SoundCue /Game/Sounds/Fort_Weapons/Rifle/ArtDeco/Weapon_Assault_ArtDeco_Impact_Dirt_Cue
root=Delay_3
nodes [
 {id=Delay_3, children [Modulator_0]}
 {id=Modulator_0, children [Random_9], properties {pitchMax=1.3, pitchMin=0.9, volumeMax=1, volumeMin=0.8}}
 {
  id=Random_9
  children [WavePlayer_53, WavePlayer_54, WavePlayer_55, WavePlayer_62, WavePlayer_63, WavePlayer_64, WavePlayer_65, WavePlayer_66]
  properties {bShouldExcludeFromBranchCulling=true, weights [1,1,1,1,1,1,1,1]}
 }
 {id=WavePlayer_53, properties {soundWaveAssetPtr=/Game/Sounds/Fort_Weapons/Impacts/Bullet_Impacts_S10/Bullet_Impact_Stone_Med_01}}
 {id=WavePlayer_54, properties {soundWaveAssetPtr=/Game/Sounds/Fort_Weapons/Impacts/Bullet_Impacts_S10/Bullet_Impact_Stone_Med_02}}
 {id=WavePlayer_55, properties {soundWaveAssetPtr=/Game/Sounds/Fort_Weapons/Impacts/Bullet_Impacts_S10/Bullet_Impact_Stone_Med_03}}
 {id=WavePlayer_62, properties {soundWaveAssetPtr=/Game/Sounds/Fort_Weapons/Impacts/Bullet_Impacts_S10/Bullet_Impact_Stone_Med_04}}
 {id=WavePlayer_63, properties {soundWaveAssetPtr=/Game/Sounds/Fort_Weapons/Impacts/Bullet_Impacts_S10/Bullet_Impact_Stone_Med_05}}
 {id=WavePlayer_64, properties {soundWaveAssetPtr=/Game/Sounds/Fort_Weapons/Impacts/Bullet_Impacts_S10/Bullet_Impact_Stone_Med_06}}
 {id=WavePlayer_65, properties {soundWaveAssetPtr=/Game/Sounds/Fort_Weapons/Impacts/Bullet_Impacts_S10/Bullet_Impact_Stone_Med_07}}
 {id=WavePlayer_66, properties {soundWaveAssetPtr=/Game/Sounds/Fort_Weapons/Impacts/Bullet_Impacts_S10/Bullet_Impact_Stone_Med_08}}
]
properties {
 attenuationSettings=/Game/Sounds/Attenuation/Fort_Bullet_Impact_Att
 maxDistance=5300
 soundClassObject=&Fort_Impacts
 volumeMultiplier=2
 concurrencyOverrides {maxCount=4, resolutionRule=PreventNew}
 concurrencySet [&RifleBulletImpact_Conc]
}
